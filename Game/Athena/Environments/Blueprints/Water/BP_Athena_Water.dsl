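; Athena water body actor that can switch between normal water and a bouncy, corrupted-looking lake.
; OnRep handlers drive material swaps and bounce mode, which makes the box solid, unwalkable and free of fall damage.
; Box hit and overlap events launch pawns and vehicles with gameplay cues, eject shopping cart riders and throttle repeat vehicle bounces by timer.
; BeginPlay builds a post-process material instance whose WaterLevel follows the water plane height.
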(event Custom|SetSpreadMaterial
  (Rendering|Material|SetMaterial (Variables|Default|GetWater_Plane) 0 (Variables|Default|GetDefaultWaterMIOverride)))

(event Custom|SetTurnedOnMaterial
  (Rendering|Material|SetMaterial (Variables|Default|GetWater_Plane) 0 (Variables|Default|GetCorruptedWaterMIOverride)))

(event Custom|SetDefaultMaterial
  (Rendering|Material|SetMaterial (Variables|Default|GetWater_Plane) 0 (Variables|Default|GetDefaultWaterMIOverride)))

(event Custom|SetBounceHeightForceMultiplier
  (params (BounceHeightMultiplier :type Float%20%28double-precision%29) (BounceVelocityMultiplier :type Float%20%28double-precision%29))
  (Variables|Default|SetBounceHeightMultiplier BounceHeightMultiplier)
  (Variables|Default|SetBounceVelocityMultiplier BounceVelocityMultiplier))

(event Custom|SetBouncy
  (params (Bouncy :type Boolean))
  (bind _box (Variables|Default|GetBox))
  (if Bouncy
    (Physics|SetWalkableSlopeOverride (Variables|Default|GetBox) (Utilities|Struct|MakeWalkableSlopeOverride "WalkableSlope_Unwalkable"))
    (for _array_element (Collision|GetOverlappingActors (Variables|Default|GetBox) "/Game/Athena/PlayerPawn_Athena.PlayerPawn_Athena_C" :!class /Script/Engine.PrimitiveComponent)
      (CallFunction|BouncePawn (Utilities|Casting|CastToPlayerPawn_Athena _array_element)))
    (Collision|SetCollisionProfileName (Variables|Default|GetBox) "FortBuildingMesh" true :!class /Script/Engine.PrimitiveComponent)
    (Variables|/Script/Engine.ActorComponent|SetComponentTags (Utilities|Array|MakeArray "NoFallDamage" "NoStick") (Variables|Default|GetBox))
    (elif (Utilities|Name|Equal(Name) (Collision|GetCollisionProfileName _box) "Custom")
      _
      (else
        (Collision|SetCollisionProfileName _box "OverlapAll" true :!class /Script/Engine.PrimitiveComponent)))))

(event Custom|Bouncy
  (Variables|Default|SetBounceLake 1))

(event OnComponentHit(Box)
  (params (HitComponent :type Primitive%20Component%20Object%20Reference) (OtherActor :type Actor%20Object%20Reference) (OtherComp :type Primitive%20Component%20Object%20Reference) (NormalImpulse :type Vector :default "0, 0, 0") (Hit :type Hit%20Result%20Structure%20%28by%20ref%29))
  (switch int (Variables|Default|GetBounceLake)
    (:1
      ;[1 1 1] Gameplay Cues
      (bind _asplayer_pawn_athena (Utilities|Casting|CastToPlayerPawn_Athena OtherActor)
        (:then
          (CallFunction|BouncePawn _asplayer_pawn_athena))
        (:CastFailed
          (bind _asfort_vehicle_interface (Utilities|Casting|CastToFortVehicleInterface OtherActor)
            (:then
              (CallFunction|BounceVehicle _asfort_vehicle_interface))
            (:CastFailed
              (Utilities|IsValid (Actor|GetComponentbyClass OtherActor "/Script/Engine.ProjectileMovementComponent")
                (:Is%20Valid
                  (CallFunction|PlayBounceGC OtherActor (Transformation|GetActorLocation OtherActor)))))))))))

(event Custom|BouncePawn
  (params (Pawn :type Player%20Pawn%20Athena%20Object%20Reference))
  ;[1 1 1] Gameplay Cues
  (Parachute|EndSkydiving Pawn)
  (Pawn|Character|LaunchCharacterJump Pawn (Math|Vector|vector+vector (Math|Vector|vector*vector (Variables|/Script/Engine.MovementComponent|GetVelocity (Variables|/Script/Engine.Character|GetCharacterMovement Pawn)) (Math|Vector|MakeVector (* (Variables|Default|GetBaseBounceVelocityForce) (Variables|Default|GetBounceVelocityMultiplier)) (* (Variables|Default|GetBaseBounceVelocityForce) (Variables|Default|GetBounceVelocityMultiplier)))) (Math|Vector|MakeVector 0.0 0.0 (* (Variables|Default|GetBaseBounceHeightForce) (Variables|Default|GetBounceHeightMultiplier)))) false true false true)
  (CallFunction|PlayBounceGC Pawn (.Location (Collision|BreakHitResult (stale OnComponentHit(Box) Hit)))))

(event Custom|PlayBounceGC
  (params (Target :type Actor%20Object%20Reference) (HitLocation :type Vector))
  ;[1 1 1] Gameplay Cues
  (Variables|Default|SetHitLocation HitLocation)
  (GameplayCue|ExecuteGameplayCueOnActor(Burst) Target (Variables|Default|GetLightningImpactCLCue) (Ability|GameplayCue|MakeGameplayCueParameters :NormalizedMagnitude 0.0 :RawMagnitude 0.0 :EffectContext "()" :AggregatedSourceTags "(GameplayTags=)" :AggregatedTargetTags "(GameplayTags=)" :Location (Math|Vector|MakeVectorNetQuantize10 (.x (Variables|Default|GetHitLocation)) (.y (Variables|Default|GetHitLocation)) (.z (Variables|Default|GetHitLocation))) :Normal (Math|Vector|MakeVectorNetQuantizeNormal 0.0 0.0 1.0) :Instigator self :GameplayEffectLevel 1 :AbilityLevel 1 :bReplicateLocationWhenUsingMinimalRepProxy false)))

(event Collision|EventActorBeginOverlap
  (params (OtherActor :type Actor%20Object%20Reference)))

(event OnComponentBeginOverlap(Box)
  (params (OverlappedComponent :type Primitive%20Component%20Object%20Reference) (OtherActor :type Actor%20Object%20Reference) (OtherComp :type Primitive%20Component%20Object%20Reference) (OtherBodyIndex :type Integer) (bFromSweep :type Boolean) (SweepResult :type Hit%20Result%20Structure%20%28by%20ref%29))
  (switch int (Variables|Default|GetBounceLake)
    (:1
      (bind _asfort_vehicle_interface (Utilities|Casting|CastToFortVehicleInterface OtherActor))
      (bind _returnvalue (Vehicle|GetVehicleActor _asfort_vehicle_interface :!class /Script/FortniteGame.FortVehicleInterface))
      (bind _asshopping_cart_vehicle_sk (Utilities|Casting|CastToShoppingCartVehicleSK _returnvalue)
        (:then
          (Vehicle|EjectAllPlayers _asshopping_cart_vehicle_sk)
          (Utilities|FlowControl|DelayUntilNextTick)
          (CallFunction|BounceVehicle _asfort_vehicle_interface))
        (:CastFailed
          (CallFunction|BounceVehicle _asfort_vehicle_interface)
          (for _array_element (Vehicle|GetAllPassengers _asfort_vehicle_interface)
            (bind _asfort_player_controller_gameplay (Utilities|Casting|CastToFortPlayerControllerGameplay (Pawn|GetController _array_element)))
            (VehicleTricks|DisableTrickCredit _asfort_player_controller_gameplay 0.1)))))))

(event EventTick
  (params (DeltaSeconds :type Float%20%28single-precision%29)))

(event Custom|ClearRecentlyBounced
  (if (> (Utilities|Array|Length (Variables|Default|GetRecentlyBouncedVehicles)) 0)
    (Utilities|Array|RemoveIndex (Variables|Default|GetRecentlyBouncedVehicles) (Utilities|Array|LastIndex (Variables|Default|GetRecentlyBouncedVehicles)))
    (Utilities|Time|SetTimerbyEvent (EventDispatchers|CreateEvent :!event Custom|ClearRecentlyBounced) 0.5)))

(event Custom|BounceVehicle
  (params (VehicleInterface :type Fort%20Vehicle%20Interface%20Interface))
  (Variables|Default|SetVehicleInterfaceToBounce VehicleInterface)
  (if (not (Utilities|Array|ContainsItem (Variables|Default|GetRecentlyBouncedVehicles) (Variables|Default|GetVehicleInterfaceToBounce)))
    (Audio|BroadcastAbilitySpinThrustatLocation 0 (Utilities|ToObject(Interface) (Variables|Default|GetVehicleInterfaceToBounce)) (+ (* (Vehicle|GetVehicleRightVector (Variables|Default|GetVehicleInterfaceToBounce)) (Math|Random|RandomFloatinRange -1.5e+02 1.5e+02)) (* (Vehicle|GetVehicleForwardVector (Variables|Default|GetVehicleInterfaceToBounce)) (Math|Random|RandomFloatinRange -1.5e+02 1.5e+02))) 7.8e+02)
    (bind _returnvalue (Vehicle|GetVehicleActor (Variables|Default|GetVehicleInterfaceToBounce) :!class /Script/FortniteGame.FortVehicleInterface))
    (Audio|BroadcastAbilityImpactatLocation "/Game/Athena/Sounds/Events/Cube/Cube_LaunchPlayer_Cue.Cube_LaunchPlayer_Cue" _returnvalue (Math|Vector|vector+vector (* (* (Math|Vector|Lerp(Vector) (Math|Vector|Normalize (Vehicle|GetVehicleVelocity (Variables|Default|GetVehicleInterfaceToBounce))) (Vehicle|GetVehicleForwardVector (Variables|Default|GetVehicleInterfaceToBounce)) 0.3) (Math|Float|Max(Float) (Math|Vector|VectorLength (Vehicle|GetVehicleVelocity (Variables|Default|GetVehicleInterfaceToBounce))) 2e+02)) "1.100000,1.100000, 0") (0.0 0.0 1.5e+03)) 7.8e+02) ; Set Physics Linear Velocity
    (CallFunction|ReplicateGCVehicles _returnvalue)
    (Utilities|Time|SetTimerbyEvent (EventDispatchers|CreateEvent :!event Custom|ClearRecentlyBounced) 0.5)))

(event Custom|ReplicateGCVehicles
  (params (Actor :type Actor%20Object%20Reference))
  (CallFunction|PlayBounceGC Actor (Transformation|GetActorLocation Actor)))

(event EventBeginPlay
  (bind _waterpost (Variables|Default|GetWaterPost))
  (bind _settings (Variables|/Script/Engine.PostProcessComponent|GetSettings _waterpost))
  (bind _weightedblendables (Utilities|Struct|BreakPostProcessSettings _settings))
  (bind _array (Utilities|Struct|BreakWeightedBlendables _weightedblendables))
  (bind (_weight _object) (Utilities|Struct|BreakWeightedBlendable (Utilities|Array|Get(aref) _array)))
  ;[1 1 1] Begin Play set PostProcess Water Level
  (Variables|Default|SetBounceLake 0)
  (if (> (Utilities|Array|Length _array) 0)
    (bind _asmaterial_interface (Utilities|Casting|CastToMaterialInterface _object)
      (:then
        _)
      (:CastFailed
        _))
    (bind _returnvalue (Rendering|Material|CreateDynamicMaterialInstance (select (Utilities|IsValid _asmaterial_interface :!class /Script/Engine.KismetSystemLibrary) _asmaterial_interface "/Game/Athena/Environments/Materials/Water/M_Athena_Water_PostProcess_0.M_Athena_Water_PostProcess_0") "None" "None" :!class /Script/Engine.KismetMaterialLibrary))
    (bind _output_get (Variables|Default|SetPP%20MID _returnvalue))
    (else
      (bind _returnvalue_2 (Rendering|Material|CreateDynamicMaterialInstance "/Game/Athena/Environments/Materials/Water/M_Athena_Water_PostProcess_0.M_Athena_Water_PostProcess_0" "None" "None" :!class /Script/Engine.KismetMaterialLibrary))
      (Variables|Default|SetPP%20MID _returnvalue_2)))
  (Rendering|Material|SetScalarParameterValue _output_get "WaterLevel" (- (.z (Transformation|GetWorldLocation (Variables|Default|GetWater_Plane))) 48.5) :!class /Script/Engine.MaterialInstanceDynamic)
  (bind _structout (Utilities|Struct|SetmembersinPostProcessSettings (Variables|/Script/Engine.PostProcessComponent|GetSettings (Variables|Default|GetWaterPost)) (Utilities|Struct|MakeWeightedBlendables (Utilities|Array|MakeArray (Utilities|Struct|MakeWeightedBlendable 1.0 (Variables|Default|GetPP%20MID))))))
  (Variables|/Script/Engine.PostProcessComponent|SetSettings _structout (Variables|Default|GetWaterPost)))

(fn OnRep_BounceLake
  (switch int (Variables|Default|GetBounceLake)
    (:0
      (CallFunction|SetBouncy false))
    (:1
      (CallFunction|SetBouncy true))))

(fn OnRep_IsBouncy
  (if (Variables|Default|GetIsBouncy)
    (CallFunction|SetBouncy false)))

(fn OnRep_WaterState
  (switch int (Variables|Default|GetWaterState)
    (:0
      (CallFunction|SetDefaultMaterial))
    (:1
      (CallFunction|SetSpreadMaterial))
    (:2
      (CallFunction|SetTurnedonMaterial))))

(fn ConstructionScript
  (bind _groundplanematerialoverride (Variables|Default|GetGroundPlaneMaterialOverride))
  (if (Variables|Default|GetHideGroundPlane)
    (Development|SetHiddeninGame (Variables|Default|GetWater_Base) true false))
  (if (Variables|Default|GetCorruptWaterMat)
    (Rendering|Material|SetMaterial (Variables|Default|GetWater_Plane) 0 (Variables|Default|GetDefaultWaterMIOverride)))
  (if (Utilities|NotEqual(Object) _groundplanematerialoverride 0)
    (Rendering|Material|SetMaterial (Variables|Default|GetWater_Base) 0 _groundplanematerialoverride)))
