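; Creative island water actor that can turn into a bouncy lake launching players and vehicles.
; BounceLake's OnRep makes the box solid, unwalkable and tagged against fall damage, bouncing pawns already inside.
; Box hits and overlaps launch pawns or spin-thrust vehicles with impact cues, using a timer-cleared list to debounce vehicles; projectiles only play the cue.
; Construction applies ground-plane and corrupt-water options, and custom events swap water materials.
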(event Custom|SetSpreadMaterial
  (Rendering|Material|SetMaterial (Variables|Default|GetWater_Plane) 0 "/Game/Packages/Fortnite_UniqueMaterials_1/DS_Fortnight/Water_Shallow_Athena_C_Inst.Water_Shallow_Athena_C_Inst"))

(event Custom|SetTurnedOnMaterial
  (Rendering|Material|SetMaterial (Variables|Default|GetWater_Plane) 0 "/Game/Packages/Fortnite_UniqueMaterials_1/DS_Fortnight/Water_Shallow_Athena_C-On_Inst.Water_Shallow_Athena_C-On_Inst"))

(event Custom|SetDefaultMaterial
  (Rendering|Material|SetMaterial (Variables|Default|GetWater_Plane) 0 "/Game/Packages/Fortnite_UniqueMaterials_1/DS_Fortnight/Water_Shallow_New_Athena.Water_Shallow_New_Athena"))

(event Custom|SetBouncy
  (params (Bouncy :type Boolean))
  (bind _box (Variables|Default|GetBox))
  (if Bouncy
    (Physics|SetWalkableSlopeOverride (Variables|Default|GetBox) (Utilities|Struct|MakeWalkableSlopeOverride "WalkableSlope_Unwalkable"))
    (for _array_element (Collision|GetOverlappingActors (Variables|Default|GetBox) "/Game/Athena/PlayerPawn_Athena.PlayerPawn_Athena_C" :!class /Script/Engine.PrimitiveComponent)
      (CallFunction|BouncePawn (Utilities|Casting|CastToPlayerPawn_Athena _array_element)))
    (Collision|SetCollisionProfileName (Variables|Default|GetBox) "FortBuildingMesh" true :!class /Script/Engine.PrimitiveComponent)
    (Variables|/Script/Engine.ActorComponent|SetComponentTags (Utilities|Array|MakeArray "NoFallDamage" "NoStick") (Variables|Default|GetBox))
    (elif (Utilities|Name|Equal(Name) (Collision|GetCollisionProfileName _box) "Custom")
      _
      (else
        (Collision|SetCollisionProfileName _box "OverlapAll" true :!class /Script/Engine.PrimitiveComponent)))))

(event Custom|Bouncy
  (Variables|Default|SetBounceLake 1))

(event Custom|BouncePawn
  (params (Pawn :type Player%20Pawn%20Athena%20Object%20Reference))
  ;[1 1 1] Gameplay Cues
  (Parachute|EndSkydiving Pawn)
  (Pawn|Character|LaunchCharacterJump Pawn (+ (* (Variables|/Script/Engine.MovementComponent|GetVelocity (Variables|/Script/Engine.Character|GetCharacterMovement Pawn)) "0.500000,0.500000,0.000000") "0, 0,2650.000000") false true false true)
  (Networking|CanExecuteCosmeticEvents
    (:True
      (Variables|Default|SetHitLocation (.Location (Collision|BreakHitResult (stale OnComponentHit(Box) Hit))))
      (CallFunction|PlayBounceGC Pawn))))

(event OnComponentHit(Box)
  (params (HitComponent :type Primitive%20Component%20Object%20Reference) (OtherActor :type Actor%20Object%20Reference) (OtherComp :type Primitive%20Component%20Object%20Reference) (NormalImpulse :type Vector :default "0, 0, 0") (Hit :type Hit%20Result%20Structure%20%28by%20ref%29))
  (switch int (Variables|Default|GetBounceLake)
    (:1
      ;[1 1 1] Gameplay Cues
      (bind _asplayer_pawn_athena (Utilities|Casting|CastToPlayerPawn_Athena OtherActor)
        (:then
          (CallFunction|BouncePawn _asplayer_pawn_athena))
        (:CastFailed
          (bind _asfort_vehicle_interface (Utilities|Casting|CastToFortVehicleInterface OtherActor)
            (:then
              (Variables|Default|SetVehicleToBounceInterface _asfort_vehicle_interface)
              (CallFunction|BounceVehicle))
            (:CastFailed
              (Utilities|IsValid (Actor|GetComponentbyClass OtherActor "/Script/Engine.ProjectileMovementComponent")
                (:Is%20Valid
                  (Variables|Default|SetHitLocation (Transformation|GetActorLocation OtherActor))
                  (GameplayCue|ExecuteGameplayCue%28Burst%29Local OtherActor (Variables|Default|GetLightningImpactCue) (Ability|GameplayCue|MakeGameplayCueParameters :NormalizedMagnitude 0.0 :RawMagnitude 0.0 :EffectContext "()" :AggregatedSourceTags "(GameplayTags=)" :AggregatedTargetTags "(GameplayTags=)" :Location (Math|Vector|MakeVectorNetQuantize10 (.x (Variables|Default|GetHitLocation)) (.y (Variables|Default|GetHitLocation)) (.z (Variables|Default|GetHitLocation))) :Normal (Math|Vector|MakeVectorNetQuantizeNormal 0.0 0.0 1.0) :Instigator self :GameplayEffectLevel 1 :AbilityLevel 1 :bReplicateLocationWhenUsingMinimalRepProxy false)))))))))))

(event Custom|PlayBounceGC
  (params (Target :type Actor%20Object%20Reference))
  ;[1 1 1] Gameplay Cues
  (GameplayCue|ExecuteGameplayCue%28Burst%29Local Target (Variables|Default|GetLightningImpactCue) (Ability|GameplayCue|MakeGameplayCueParameters :NormalizedMagnitude 0.0 :RawMagnitude 0.0 :EffectContext "()" :AggregatedSourceTags "(GameplayTags=)" :AggregatedTargetTags "(GameplayTags=)" :Location (Math|Vector|MakeVectorNetQuantize10 (.x (Variables|Default|GetHitLocation)) (.y (Variables|Default|GetHitLocation)) (.z (Variables|Default|GetHitLocation))) :Normal (Math|Vector|MakeVectorNetQuantizeNormal 0.0 0.0 1.0) :Instigator self :GameplayEffectLevel 1 :AbilityLevel 1 :bReplicateLocationWhenUsingMinimalRepProxy false)))

(event EventBeginPlay
  (Variables|Default|SetBounceLake 0))

(event Collision|EventActorBeginOverlap
  (params (OtherActor :type Actor%20Object%20Reference)))

(event OnComponentBeginOverlap(Box)
  (params (OverlappedComponent :type Primitive%20Component%20Object%20Reference) (OtherActor :type Actor%20Object%20Reference) (OtherComp :type Primitive%20Component%20Object%20Reference) (OtherBodyIndex :type Integer) (bFromSweep :type Boolean) (SweepResult :type Hit%20Result%20Structure%20%28by%20ref%29))
  (switch int (Variables|Default|GetBounceLake)
    (:1
      (bind _asfort_vehicle_interface (Utilities|Casting|CastToFortVehicleInterface OtherActor))
      (bind _output_get (Variables|Default|SetVehicleToBounceInterface _asfort_vehicle_interface))
      (CallFunction|BounceVehicle)
      (for _array_element (Vehicle|GetAllPassengers _output_get)
        (bind _asfort_player_controller_gameplay (Utilities|Casting|CastToFortPlayerControllerGameplay (Pawn|GetController _array_element)))
        (VehicleTricks|DisableTrickCredit _asfort_player_controller_gameplay 0.1)))))

(event EventTick
  (params (DeltaSeconds :type Float%20%28single-precision%29)))

(event Custom|ClearRecentlyBounced
  (if (> (Utilities|Array|Length (Variables|Default|GetRecentlyBouncedVehicles)) 0)
    (Utilities|Array|RemoveIndex (Variables|Default|GetRecentlyBouncedVehicles) (Utilities|Array|LastIndex (Variables|Default|GetRecentlyBouncedVehicles)))
    (Utilities|Time|SetTimerbyEvent (EventDispatchers|CreateEvent :!event Custom|ClearRecentlyBounced) 0.5)))

(event Custom|BounceVehicle
  (bind _vehicletobounceinterface (Variables|Default|GetVehicleToBounceInterface))
  (bind _vehicletobounceinterface_2 (Variables|Default|GetVehicleToBounceInterface))
  (if (not (Utilities|Array|ContainsItem (Variables|Default|GetRecentlyBouncedVehicles) (Variables|Default|GetVehicleToBounceInterface)))
    (bind _returnvalue (Vehicle|GetVehicleActor _vehicletobounceinterface :!class /Script/FortniteGame.FortVehicleInterface))
    (Audio|BroadcastAbilitySpinThrustatLocation 0 _returnvalue (+ (* (Vehicle|GetVehicleRightVector _vehicletobounceinterface) (Math|Random|RandomFloatinRange -1.5e+02 1.5e+02)) (* (Vehicle|GetVehicleForwardVector _vehicletobounceinterface) (Math|Random|RandomFloatinRange -1.5e+02 1.5e+02))) 7.8e+02)
    (bind _returnvalue_2 (Vehicle|GetVehicleActor _vehicletobounceinterface_2 :!class /Script/FortniteGame.FortVehicleInterface))
    (Audio|BroadcastAbilityImpactatLocation "/Game/Athena/Sounds/Events/Cube/Cube_LaunchPlayer_Cue.Cube_LaunchPlayer_Cue" _returnvalue_2 (Math|Vector|vector+vector (* (* (Math|Vector|Lerp(Vector) (Math|Vector|Normalize (Vehicle|GetVehicleVelocity _vehicletobounceinterface_2)) (Vehicle|GetVehicleForwardVector _vehicletobounceinterface_2) 0.3) (Math|Float|Max(Float) (Math|Vector|VectorLength "0, 0, 0") 2e+02)) "1.100000,1.100000, 0") (0.0 0.0 1.5e+03)) 7.8e+02) ; Set Physics Linear Velocity
    (CallFunction|ReplicateGCVehicles _returnvalue_2)
    (Utilities|Time|SetTimerbyEvent (EventDispatchers|CreateEvent :!event Custom|ClearRecentlyBounced) 0.5)))

(event Custom|ReplicateGCVehicles
  (params (Actor :type Actor%20Object%20Reference))
  (Variables|Default|SetHitLocation (Transformation|GetActorLocation Actor))
  (CallFunction|PlayBounceGC Actor))

(fn OnRep_BounceLake
  (switch int (Variables|Default|GetBounceLake)
    (:0
      (CallFunction|SetBouncy false))
    (:1
      (CallFunction|SetBouncy true))))

(fn OnRep_IsBouncy
  (if (Variables|Default|GetIsBouncy)
    (CallFunction|SetBouncy false)))

(fn ConstructionScript
  (bind _groundplanematerialoverride (Variables|Default|GetGroundPlaneMaterialOverride))
  (if (Variables|Default|GetHideGroundPlane)
    (Development|SetHiddeninGame (Variables|Default|GetWater_Base) true false))
  (if (Variables|Default|GetCorruptWaterMat)
    (Rendering|Material|SetMaterial (Variables|Default|GetWater_Plane) 0 "/Game/Packages/Fortnite_UniqueMaterials_1/DS_Fortnight/Water_Shallow_Athena_C_Inst.Water_Shallow_Athena_C_Inst"))
  (if (Utilities|NotEqual(Object) _groundplanematerialoverride 0)
    (Rendering|Material|SetMaterial (Variables|Default|GetWater_Base) 0 _groundplanematerialoverride)))
